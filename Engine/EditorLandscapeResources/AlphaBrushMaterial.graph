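# Unlit translucent landscape editor material that previews an alpha texture brush with depth testing disabled.
# World position relative to the brush is scaled and rotated by parameters into UVs for a channel-masked texture sample, clipped to the brush square.
# Screen-space derivatives size a border line, which adds to the faded alpha.
# CanPaint blends the border color from dark red, with inverse eye adaptation on higher feature levels.
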Material /Engine/EditorLandscapeResources/AlphaBrushMaterial
graph {
 properties {
  bDisableDepthTest=true
  blendMode=BLEND_Translucent
  pixelDepthOffsetMode=PDOM_Legacy
  shadingModel=MSM_Unlit
  translucencyPass=MTP_BeforeDOF
  usedShadingModels=Unlit
 }
 nodes [
  n1 = NamedRerouteUsage()
  n2 = NamedRerouteUsage()
  n3 = NamedRerouteUsage()
  n4 = NamedRerouteUsage()
  n5 = Add(A=OneMinus(None=Step(Y=@n1, X=@n2)), B=OneMinus(None=Step(Y=@n3, X=@n4)))
  n6 = NamedRerouteUsage()
  n7 = NamedRerouteUsage()
  n8 = Add(A=Step(Y=OneMinus(None=@n6), X=NamedRerouteUsage()), B=Step(Y=OneMinus(None=@n7), X=NamedRerouteUsage()))
  n9 = Multiply(
   A=MaterialFunctionCall!(RangeMin, RangeMax, Value=NamedRerouteUsage()) {materialFunction=/Engine/Functions/Engine_MaterialFunctions03/Math/WithinRangeFloat2}.AllComponentsWithinRange
   B=Saturate(None=Add(A=@n5, B=@n8))
  )
  n10 = AppendVector(A=AppendVector(A=@n9, B=@n9), B=@n9)
  Add(A=@n10) {constB=0}
  n11 = LinearInterpolate(A=Constant3Vector() {constant=<rgba 0.2 0 0 0>}, B=@n10, Alpha=ScalarParameter() {defaultValue=1, parameterName=CanPaint})
  n12 = EyeAdaptationInverse(LightValueInput=@n11)
  FeatureLevelSwitch(
   Default=@n12.EyeAdaptationInverse
   ES2=@n11
   ES3_1=@n11
   SM4_REMOVED=@n12.EyeAdaptationInverse
   SM5=@n12.EyeAdaptationInverse
   SM6=@n12.EyeAdaptationInverse
  )
  n13 = WorldPosition()
  n14 = DoubleVectorParameter() {desc=Brush world position, parameterName=WorldPosition, defaultValue=<0 0 0 0>}
  n15 = ComponentMask(None=Transform(None=Subtract(A=@n13.XYZ, B=@n14.XYZ)) {transformSourceType=TRANSFORMSOURCE_World, transformType=TRANSFORM_Local}) {g=true, r=true}
  n16 = VectorParameter() {parameterName=BrushScaleRot, defaultValue=<rgba 0.001 0.001 0 0.5>}
  n17 = AppendVector(A=@n16.R, B=@n16.R)
  n18 = Add(A=Rotator(Coordinate=Multiply(A=@n15, B=@n17), Time=@n16.A) {centerX=0, centerY=0, speed=1}, B=Constant2Vector() {g=0.5, r=0.5})
  n19 = VectorParameter() {parameterName=AlphaTextureMask, defaultValue=<rgba 0 0 0 1>}
  n20 = AppendVector(A=@n19.RGB, B=@n19.RGB)
  n21 = TextureSampleParameter2D(UVs=@n18) {parameterName=AlphaTexture, texture=&/DefaultAlphaTexture}
  n22 = AppendVector(A=@n21.RGB, B=@n21.RGB)
  n23 = Multiply(
   A=ScalarParameter() {defaultValue=0.5, parameterName=OpacityAmount}
   B=Multiply(
    A=MaterialFunctionCall(Value=@n18) {materialFunction=&WithinRangeFloat2}.AllComponentsWithinRange
    B=DotProduct(A=@n20, B=@n22)
   )
  )
  Multiply(
   A=AppendVector(A=AppendVector(A=@n23, B=@n23), B=@n23)
   B=ComponentMask(None=VectorParameter() {parameterName=HighlightColor, defaultValue=<rgba 1 1 1 1>}.RGB) {b=true, g=true, r=true}
  )
  n24 = NamedRerouteDeclaration(None=@n18) {name=Coords, nodeColor=<rgba 0 0.7656 1 1>}
  NamedRerouteDeclaration(None=ComponentMask(None=@n24) {r=true}) {name=X, nodeColor=<rgba 1 0.5391 0 1>}
  NamedRerouteDeclaration(None=ComponentMask(None=@n24) {g=true}) {name=Y, nodeColor=<rgba 0 1 0.8125 1>}
  n25 = NamedRerouteUsage()
  n26 = DDX(Value=@n25)
  n27 = DDY(Value=@n25)
  NamedRerouteDeclaration(None=Min(
    A=MaterialFunctionCall(Vector 2=@n26) {materialFunction=/Engine/Functions/Engine_MaterialFunctions02/Utility/VectorLength}.V2 Length
    B=MaterialFunctionCall(Vector 2=@n27) {materialFunction=&VectorLength}.V2 Length
   )) {name=LineThickness, nodeColor=<rgba 0 1 0.1797 1>}
  Reroute(None=Add(A=@n9, B=Multiply(A=@n23) {constB=0.8}))
 ]
}
