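# Material function in the landscape editor resources that computes an outline of screen-relative thickness where a Distance value equals a Radius.
# Thickness is the smaller length of the screen-space derivatives (DDX, DDY) of LocalVector, scaled by ThicknessMultiplicator, keeping line width roughly constant in pixels.
# The absolute gap between Distance and Radius is divided by that thickness, raised to a power and inverted, without clamping.
# It outputs both Outline and Thickness.
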MaterialFunction /Engine/EditorLandscapeResources/MF_ThicknessOutline
graph {nodes [
  n1 = FunctionInput() {bUsePreviewValueAsDefault=true, inputName=Distance, inputType=FunctionInput_Scalar}
  n2 = FunctionInput() {bUsePreviewValueAsDefault=true, inputName=Radius, inputType=FunctionInput_Scalar}
  n3 = FunctionInput() {bUsePreviewValueAsDefault=true, inputName=LocalVector, inputType=FunctionInput_Vector2}
  n4 = DDX(Value=@n3)
  n5 = DDY(Value=@n3)
  n6 = FunctionInput() {bUsePreviewValueAsDefault=true, inputName=ThicknessMultiplicator, inputType=FunctionInput_Scalar, previewValue=<1 0 0 1>}
  n7 = Multiply(
   A=Min(
    A=MaterialFunctionCall(Vector 2=@n4) {materialFunction=/Engine/Functions/Engine_MaterialFunctions02/Utility/VectorLength}.V2 Length
    B=MaterialFunctionCall(Vector 2=@n5) {materialFunction=&VectorLength}.V2 Length
   )
   B=@n6
  ) {constB=2}
  n8 = Divide(A=Abs(None=Subtract(A=@n1, B=@n2)), B=@n7)
  FunctionOutput(None=OneMinus(None=Power(Base=@n8))) {outputName=Outline}
  FunctionOutput(None=@n7) {outputName=Thickness}
 ]}
