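# Unlit translucent material that previews the landscape editor's circular brush, drawn without depth testing.
# Local-space distance from brush position, radius and falloff drives inner and outer outlines, with thickness from a material function.
# Static switches select smooth, spherical, tip or linear falloff and an optional rotated, channel-masked alpha texture, scaled by opacity.
# CanPaint blends outline color from dark red, with inverse eye adaptation keeping brightness stable on higher feature levels.
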Material /Engine/EditorLandscapeResources/CircleBrushMaterial
graph {
 properties {
  bDisableDepthTest=true
  bEnableResponsiveAA=true
  blendMode=BLEND_Translucent
  pixelDepthOffsetMode=PDOM_Legacy
  shadingModel=MSM_Unlit
  translucencyPass=MTP_BeforeDOF
  usedShadingModels=Unlit
 }
 nodes [
  n1 = Subtract(A=NamedRerouteUsage(), B=NamedRerouteUsage())
  n2 = Subtract(A=NamedRerouteUsage(), B=Add(A=NamedRerouteUsage(), B=NamedRerouteUsage()))
  n3 = Constant()
  n4 = Divide(A=NamedRerouteUsage(), B=NamedRerouteUsage())
  n5 = If(
   A=NamedRerouteDeclaration(None=Abs(None=@n2)) {name=Outer, nodeColor=<rgba 1 0.5391 0 1>}
   B=NamedRerouteUsage()
   "A > B"=@n3
   "A == B"=@n3
   "A < B"=OneMinus(None=Power(Base=@n4))
  )
  n6 = Divide(A=NamedRerouteUsage(), B=NamedRerouteUsage())
  n7 = If(
   A=NamedRerouteDeclaration(None=Abs(None=@n1)) {name=Inner, nodeColor=<rgba 0 0.6016 1 1>}
   B=NamedRerouteUsage()
   "A > B"=@n5
   "A == B"=@n5
   "A < B"=OneMinus(None=Power(Base=@n6))
  )
  n8 = Power(Base=NamedRerouteUsage())
  n9 = OneMinus(None=SquareRoot(None=OneMinus(None=@n8)))
  n10 = NamedRerouteUsage()
  n11 = Multiply(B=@n10) {constA=2}
  n12 = OneMinus(None=Power(Base=NamedRerouteUsage()))
  n13 = StaticSwitchParameter(
   True=Multiply(A=Power(Base=@n10), B=Subtract(B=@n11) {constA=3})
   False=StaticSwitchParameter(True=SquareRoot(None=@n12), False=NamedRerouteUsage()) {parameterName=Sphere}
  ) {parameterName=smooth}
  n14 = VectorParameter() {parameterName=AlphaTextureMask, defaultValue=<rgba 0 0 0 1>}
  n15 = WorldPosition()
  n16 = DoubleVectorParameter() {parameterName=LandscapeLocation, defaultValue=<0 0 0 0>}
  n17 = Subtract(A=@n15.XYZ, B=@n16.XYZ)
  n18 = VectorParameter() {parameterName=AlphaScaleBias, defaultValue=<rgba 0.0001 0.0001 0 0>}
  n19 = Multiply(
   A=ComponentMask(None=Transform(None=@n17) {transformSourceType=TRANSFORMSOURCE_World, transformType=TRANSFORM_Local}) {g=true, r=true}
   B=AppendVector(A=@n18.R, B=@n18.R)
  )
  n20 = AppendVector(A=@n18.B, B=@n18.B)
  n21 = TextureSampleParameter2D(UVs=Rotator(Coordinate=Add(A=@n19, B=@n20), Time=@n16.W) {centerX=0, centerY=0, speed=1}) {parameterName=AlphaTexture, texture=&/DefaultAlphaTexture}
  n22 = DotProduct(A=AppendVector(A=@n14.RGB, B=@n14.RGB), B=AppendVector(A=@n21.RGB, B=@n21.RGB))
  n23 = Constant() {r=1}
  n24 = Multiply(
   A=Multiply(A=ScalarParameter() {defaultValue=0.5, parameterName=OpacityAmount})
   B=Multiply(A=StaticSwitchParameter(True=@n9, False=@n13) {parameterName=Tip}, B=StaticSwitchParameter(True=@n22, False=@n23) {parameterName=AlphaBrush})
  )
  Add(A=@n7, B=Multiply(B=@n24))
  n25 = LinearInterpolate(A=Constant3Vector() {constant=<rgba 0.2 0 0 0>}, B=@n7, Alpha=ScalarParameter() {defaultValue=1, parameterName=CanPaint})
  n26 = EyeAdaptationInverse(LightValueInput=@n25)
  FeatureLevelSwitch(
   Default=@n26.EyeAdaptationInverse
   ES2=@n25
   ES3_1=@n25
   SM4_REMOVED=@n26.EyeAdaptationInverse
   SM5=@n26.EyeAdaptationInverse
   SM6=@n26.EyeAdaptationInverse
  )
  Multiply(A=OneMinus(None=@n24), B=Saturate(None=Multiply(A=@n24) {constB=500}))
  NamedRerouteDeclaration(None=ScalarParameter() {defaultValue=50, parameterName=LocalRadius}) {name=Radius, nodeColor=<rgba 0 0.1562 1 1>}
  NamedRerouteDeclaration(None=ScalarParameter() {defaultValue=100, parameterName=LocalFalloff}) {name=Falloff, nodeColor=<rgba 0 0.1562 1 1>}
  n27 = Subtract(A=DoubleVectorParameter() {parameterName=WorldPosition, defaultValue=<0 0 0 0>}.XYZ, B=@n15.XYZ)
  n28 = ComponentMask(None=NamedRerouteDeclaration(None=Transform(None=@n27) {transformSourceType=TRANSFORMSOURCE_World, transformType=TRANSFORM_Local}) {name=LocalVector, nodeColor=<rgba 0.9453 0 1 1>}) {g=true, r=true}
  NamedRerouteDeclaration(None=SquareRoot(None=DotProduct(A=@n28, B=@n28))) {name=Distance, nodeColor=<rgba 0 0.1562 1 1>}
  n29 = NamedRerouteUsage()
  NamedRerouteDeclaration(None=MaterialFunctionCall!(Distance, Radius, ThicknessMultiplicator, LocalVector=ComponentMask(None=@n29) {g=true, r=true}) {materialFunction=&/MF_ThicknessOutline}.Thickness) {name=LineThickness, nodeColor=<rgba 0 0.5312 1 1>}
  n30 = NamedRerouteUsage()
  n31 = MaterialFunctionCall!(Input=NamedRerouteUsage(), Input Low=@n30, Input High=Add(A=@n30, B=NamedRerouteUsage())) {materialFunction=/Engine/Functions/Engine_MaterialFunctions03/Math/RemapValueRangeNormalized}
  NamedRerouteDeclaration(None=OneMinus(None=NamedRerouteDeclaration(None=@n31.Result) {name=FalloffDistanceRatio, nodeColor=<rgba 0.3828 0 1 1>})) {name=OneMinusFalloffDistanceRatio, nodeColor=<rgba 0.4766 1 0 1>}
  Saturate(None=Add(A=@n7, B=@n24))
  StaticSwitchParameter() {parameterName=AlphaBrush}
  VectorParameter() {parameterName=HighlightColor, defaultValue=<rgba 1 1 1 1>}
 ]
}
