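# Editor material for the landscape color mask brush, an unlit translucent surface that visualizes masked landscape data with a highlight color.
# It samples the red channel of the SelectedData texture using landscape weightmap UVs, and an If node driven by bInverted chooses that value or its inverse.
# The result is clamped to a raised floor and inverted, capping the mask below full strength.
# HighlightColor, orange by default, supplies the tint parameter.
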Material /Engine/EditorLandscapeResources/ColorMaskBrushMaterial
graph {
 properties {bUsedWithStaticLighting=true, blendMode=BLEND_Translucent, pixelDepthOffsetMode=PDOM_Legacy, shadingModel=MSM_Unlit, usedShadingModels=Unlit}
 nodes [
  n1 = ScalarParameter() {parameterName=bInverted}
  n2 = TextureSampleParameter2D(UVs=LandscapeLayerCoords() {customUVType=LCCT_WeightMapUV}) {parameterName=SelectedData, texture=&/DefaultAlphaTexture}
  n3 = OneMinus(None=@n2.R)
  OneMinus(None=Clamp(None=If(A=@n1, "A > B"=@n3, "A == B"=@n2.R, "A < B"=@n2.R)) {minDefault=0.35})
  VectorParameter() {parameterName=HighlightColor, defaultValue=<rgba 1 0.1577 0 1>}
 ]
}
