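# Editor debug material that displays landscape layer weights as colors.
# Three chained LandscapeLayerWeight nodes named R, G and B blend constant red, green and blue by layer weight, each with an equal preview weight.
# The blended color feeds the DiffuseColor input of the ConvertFromDiffSpec material function in a Default Lit material.
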Material /Engine/EditorLandscapeResources/DebugLayerColorMat
graph {
 properties {bUsedWithStaticLighting=true, pixelDepthOffsetMode=PDOM_Legacy, usedShadingModels=Default Lit}
 nodes [
  MaterialExpressionTerrainLayerWeight_0 = LandscapeLayerWeight!(Base, Layer=Constant3Vector() {constant=<rgba 1 0 0 0>}) {parameterName=R, previewWeight=0.3333}
  MaterialExpressionTerrainLayerWeight_1 = LandscapeLayerWeight(Base=@MaterialExpressionTerrainLayerWeight_0, Layer=Constant3Vector() {constant=<rgba 0 1 0 0>}) {parameterName=G, previewWeight=0.3333}
  MaterialExpressionTerrainLayerWeight_2 = LandscapeLayerWeight(Base=@MaterialExpressionTerrainLayerWeight_1, Layer=Constant3Vector() {constant=<rgba 0 0 1 0>}) {parameterName=B, previewWeight=0.3333}
  MaterialFunctionCall(DiffuseColor=@MaterialExpressionTerrainLayerWeight_2) {materialFunction=&ConvertFromDiffSpec}
 ]
}
