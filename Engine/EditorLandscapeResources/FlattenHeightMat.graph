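# Translucent, unlit, two-sided editor material for landscape height flattening, judging by its name and folder, that builds a radially faded grid pattern.
# UVs tiled by the Scale parameter are floored and compared with derivative-offset copies to mark cell edges, with feature-level switches around DDX/DDY.
# A cosine-shaped radial gradient multiplies that line mask.
# Alternative falloff branches are unused, and no material output connection appears.
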Material /Engine/EditorLandscapeResources/FlattenHeightMat
graph {
 properties {
  bUsedWithNanite=true
  blendMode=BLEND_Translucent
  pixelDepthOffsetMode=PDOM_Legacy
  shadingModel=MSM_Unlit
  translucencyPass=MTP_BeforeDOF
  twoSided=true
  usedShadingModels=Unlit
 }
 nodes [
  Constant3Vector() {constant=<rgba 0.1 0.1 0.1 0>}
  n1 = MaterialFunctionCall!(Radius, CenterPosition) {materialFunction=/Engine/Functions/Engine_MaterialFunctions01/Gradient/RadialGradient}
  ConstantBiasScale(None=@n1.RadialGradient) {bias=0, desc=Linear, scale=1}
  n2 = OneMinus(None=@n1.RadialGradient)
  ConstantBiasScale(None=Cosine(None=@n2) {period=2}) {desc=Cos Soft}
  n3 = VectorParameter() {parameterName=Scale, defaultValue=<rgba 1 1 0 0>}
  n4 = Multiply(A=TextureCoordinate(), B=Multiply(A=ComponentMask(None=@n3.RGB) {g=true, r=true}, B=Constant() {r=100}))
  n5 = Floor(None=@n4)
  n6 = DDX(Value=@n4)
  n7 = Constant()
  n8 = FeatureLevelSwitch(Default=@n6, ES2=@n7, ES3_1=@n7, SM4_REMOVED=@n6)
  n9 = Subtract(A=@n5, B=Floor(None=Add(A=@n4, B=@n8)))
  n10 = DDY(Value=@n4)
  n11 = FeatureLevelSwitch(Default=@n10, ES2=Constant(), SM4_REMOVED=@n10)
  n12 = Subtract(A=@n5, B=Floor(None=Add(A=@n4, B=@n11)))
  n13 = Add(A=DotProduct(A=@n9, B=@n9), B=DotProduct(A=@n12, B=@n12))
  n14 = Subtract(A=@n5, B=Floor(None=Add(A=@n4, B=@n11)))
  n15 = DotProduct(A=@n14, B=@n14)
  n16 = Constant()
  n17 = Constant() {r=1}
  Multiply(A=If(A=Add(A=@n13, B=@n15), B=@n16, "A > B"=@n17, "A == B"=@n16, "A < B"=@n17), B=Cosine(None=@n2) {desc=Cos Hard, period=4})
 ]
}
