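# Editor foliage brush sphere material: an unlit, additive, two-sided translucent volume with depth test disabled that highlights scene surfaces inside the brush.
# World position behind translucency, measured from the scaled object centre, forms a back-face-only mask that tints desaturated scene color with HighlightColor.
# Depth-masked Fresnel, OpacityAmount and DepthFade intersection terms are added, with a time-based sine pulse modulating intensity.
# The graph also contains an eight-tap scene-depth edge detection branch.
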Material /Engine/EditorLandscapeResources/FoliageBrushSphereMaterial
graph {
 properties {
  bDisableDepthTest=true
  bTangentSpaceNormal=false
  blendMode=BLEND_Additive
  pixelDepthOffsetMode=PDOM_Legacy
  shadingModel=MSM_Unlit
  twoSided=true
  usedShadingModels=Unlit
 }
 nodes [
  Add()
  Constant()
  n1 = Subtract(A=SceneDepth().R, B=PixelDepth().R)
  Divide(A=@n1) {constB=8}
  n2 = Clamp(None=@n1)
  n3 = OneMinus(None=@n2)
  n4 = StaticBool()
  MaterialFunctionCall(Value=@n3, -1 to 1=@n4, Sine Phase=@n4) {materialFunction=&LinearSine}
  Multiply()
  n5 = Time() {bIgnorePause=true}
  n6 = Constant() {r=1}
  n7 = ConstantBiasScale(None=Sine(None=Multiply(A=@n5, B=@n6))) {bias=-1.5, scale=0.2}
  n8 = VectorParameter() {parameterName=HighlightColor, defaultValue=<rgba 0.5 0.5 1 1>}
  n9 = MaterialFunctionCall() {materialFunction=&WorldPositionBehindTranslucency}
  n10 = ObjectPositionWS()
  n11 = MaterialFunctionCall() {materialFunction=/Engine/Functions/Engine_MaterialFunctions02/WorldPositionOffset/ObjectScale}
  n12 = Clamp(None=Divide(A=Distance(A=@n9.Result, B=@n10), B=Multiply(B=@n11.Scale XYZ) {constA=160, constB=0.75}))
  n13 = OneMinus(None=@n12)
  n14 = Multiply(A=@n8.RGB, B=Multiply(A=@n12, B=Ceil(None=@n13)))
  n15 = Abs(None=Clamp(None=TwoSidedSign()) {desc=1 on back side, maxDefault=0, minDefault=-1})
  n16 = SceneColor!(UVs)
  n17 = Constant()
  n18 = Desaturation(None=ComponentMask(None=FeatureLevelSwitch(Default=@n16.RGB, ES2=@n17, ES3_1=@n17, SM4_REMOVED=@n16.RGB) {desc=Temp fix for ES2 error}) {b=true, g=true, r=true})
  n19 = Multiply(A=Multiply(A=Fresnel()) {constB=0.1}) {constB=-0.5}
  n20 = Clamp(None=Subtract(A=SceneDepth().R, B=PixelDepth().R))
  n21 = Subtract(A=PixelDepth().R) {constA=32, constB=32}
  n22 = Multiply(A=ScalarParameter() {defaultValue=0.05, parameterName=OpacityAmount}, B=@n8.RGB) {constA=0.05}
  n23 = OneMinus(None=DepthFade())
  n24 = Clamp(None=Add(A=Add() {constB=0}, B=Multiply(A=@n23) {constB=0.05}))
  n25 = Multiply(A=Clamp(None=Divide(A=@n21) {constB=256}), B=Multiply(A=Add(A=@n22, B=@n24), B=@n20))
  n26 = Add(A=Multiply(A=Multiply(A=@n14, B=@n15), B=Multiply(A=@n18, B=@n8.RGB)), B=Add(A=Multiply(A=@n19, B=@n20), B=Clamp(None=@n25)))
  Multiply(A=Multiply(A=OneMinus(None=@n7), B=@n26)) {constB=2, desc=intensity}
  n27 = PixelDepth()
  n28 = MaterialFunctionCall!(OffsetAmount=Constant() {r=5}) {materialFunction=/Engine/Functions/Engine_MaterialFunctions01/ImageAdjustment/BlurSampleOffsets}
  n29 = Constant() {r=1}
  n30 = Divide(
   A=Subtract(
    A=@n27.R
    B=MaterialFunctionCall!(Fraction Offset, Use Custom Depth, Pixel Offset=@n28."0,1") {materialFunction=/Engine/Functions/Engine_MaterialFunctions02/SampleSceneDepth}.Depth Using Fraction Offset
   )
   B=@n29
  )
  n31 = Divide(
   A=Subtract(
    A=@n27.R
    B=MaterialFunctionCall(Pixel Offset=@n28."1,0") {materialFunction=&SampleSceneDepth}.Depth Using Pixel Offset
   )
   B=@n29
  )
  n32 = Add(A=Ceil(None=Clamp(None=@n30)), B=Ceil(None=Clamp(None=@n31)))
  n33 = Divide(
   A=Subtract(
    A=@n27.R
    B=MaterialFunctionCall(Pixel Offset=@n28."0,-1") {materialFunction=&SampleSceneDepth}.Depth Using Pixel Offset
   )
   B=@n29
  )
  n34 = Divide(
   A=Subtract(
    A=@n27.R
    B=MaterialFunctionCall(Pixel Offset=@n28."-1,0") {materialFunction=&SampleSceneDepth}.Depth Using Pixel Offset
   )
   B=@n29
  )
  n35 = Add(A=Ceil(None=Clamp(None=@n33)), B=Ceil(None=Clamp(None=@n34)))
  n36 = Divide(
   A=Subtract(
    A=@n27.R
    B=MaterialFunctionCall(Pixel Offset=@n28."-.5,.5") {materialFunction=&SampleSceneDepth}.Depth Using Fraction Offset
   )
   B=@n29
  )
  n37 = Divide(
   A=Subtract(
    A=@n27.R
    B=MaterialFunctionCall(Pixel Offset=@n28.".5,.5") {materialFunction=&SampleSceneDepth}.Depth Using Pixel Offset
   )
   B=@n29
  )
  n38 = Add(A=Ceil(None=Clamp(None=@n36)), B=Ceil(None=Clamp(None=@n37)))
  n39 = Divide(
   A=Subtract(
    A=@n27.R
    B=MaterialFunctionCall(Pixel Offset=@n28.".5,-.5") {materialFunction=&SampleSceneDepth}.Depth Using Pixel Offset
   )
   B=@n29
  )
  n40 = Divide(
   A=Subtract(
    A=@n27.R
    B=MaterialFunctionCall(Pixel Offset=@n28."-.5,-.5") {materialFunction=&SampleSceneDepth}.Depth Using Pixel Offset
   )
   B=@n29
  )
  n41 = Add(A=Ceil(None=Clamp(None=@n39)), B=Ceil(None=Clamp(None=@n40)))
  Multiply(A=Add(A=Add(A=@n32, B=@n35), B=Add(A=@n38, B=@n41))) {constB=0.1}
  Multiply(A=Multiply(A=@n2, B=@n3), B=Constant3Vector() {constant=<rgba 0.0871 0.1335 0.2655 0>})
  Multiply(A=@n26) {constB=0.5}
 ]
}
