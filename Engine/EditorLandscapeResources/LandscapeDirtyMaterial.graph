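# Translucent unlit landscape editor overlay material that color-codes dirty terrain data by the kind of change.
# It samples the SelectedData texture at landscape weightmap UVs and scales red by 255, flooring it to an integer code.
# Chained If nodes act as a switch, mapping codes to parameter colors for heightmap, weightmap or combined changes.
# Any other code falls through to InvalidDiffColor, black by default.
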Material /Engine/EditorLandscapeResources/LandscapeDirtyMaterial
graph {
 properties {blendMode=BLEND_Translucent, pixelDepthOffsetMode=PDOM_Legacy, shadingModel=MSM_Unlit, usedShadingModels=Unlit}
 nodes [
  Constant() {r=0.9}
  n1 = LandscapeLayerCoords() {customUVType=LCCT_WeightMapUV}
  n2 = Floor(None=Multiply(A=TextureSampleParameter2D(UVs=@n1) {parameterName=SelectedData, texture=&/DefaultAlphaTexture}.R) {constB=255})
  n3 = VectorParameter() {parameterName=InvalidDiffColor, defaultValue=<rgba 0 0 0 1>}
  n4 = If(
   A=@n2
   "A > B"=@n3.RGB
   "A == B"=VectorParameter() {parameterName=HeightmapDiffColor, defaultValue=<rgba 0.8796 0.0908 0.117 1>}.RGB
   "A < B"=@n3.RGB
  ) {constB=1, equalsThreshold=0.1}
  n5 = If(
   A=@n2
   "A > B"=@n4
   "A == B"=VectorParameter() {parameterName=WeightmapDiffColor, defaultValue=<rgba 0.192 0.8796 0.1378 1>}.RGB
   "A < B"=@n4
  ) {constB=2, equalsThreshold=0.1}
  If(
   A=@n2
   "A > B"=@n5
   "A == B"=VectorParameter() {parameterName=HeightmapAndWeightmapDiffColor, defaultValue=<rgba 0.8796 0.7937 0 1>}.RGB
   "A < B"=@n5
  ) {constB=3, equalsThreshold=0.1}
 ]
}
