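# Masked two-sided editor material for the landscape gizmo height display, drawing an animated wire pattern with a square border.
# MatrixRow parameters transform world position into local coordinates for derivative-thinned lines, broken up by time-scrolled sine dashes.
# These merge with a UV outline into a mask, replaced by a constant on ES2/ES3.1.
# WireColor gets a view-angle highlight, and a camera-distance-scaled offset is computed along the surface normal.
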Material /Engine/EditorLandscapeResources/LandscapeGizmoHeight_Mat
graph {
 properties {bUsedWithStaticLighting=true, blendMode=BLEND_Masked, pixelDepthOffsetMode=PDOM_Legacy, twoSided=true, usedShadingModels=Default Lit}
 nodes [
  Constant()
  Constant() {r=0.3}
  Constant() {r=0.5}
  n1 = VectorParameter() {parameterName=MatrixRow3, defaultValue=<rgba 0 0 0.0078 0>}
  n2 = VectorParameter() {parameterName=MatrixRow4, defaultValue=<rgba 0 0 0 1>}
  n3 = VectorParameter() {parameterName=MatrixRow2, defaultValue=<rgba 0 0.0078 0 0>}
  n4 = ComponentMask(None=WorldPosition().XYZ) {b=true, g=true, r=true}
  n5 = Constant() {r=1}
  n6 = VectorParameter() {parameterName=MatrixRow1, defaultValue=<rgba 0.0078 0 0 0>}
  n7 = ComponentMask(None=MaterialFunctionCall!(
    MatrixRow3=AppendVector(A=@n1.RGB, B=@n1.RGB)
    MatrixRow4=AppendVector(A=@n2.RGB, B=@n2.RGB)
    MatrixRow2=AppendVector(A=@n3.RGB, B=@n3.RGB)
    Vector=AppendVector(A=@n4, B=@n5)
    MatrixRow1=AppendVector(A=@n6.RGB, B=@n6.RGB)
   ) {materialFunction=/Engine/Functions/Engine_MaterialFunctions02/Math/MulM4V4}.Result) {g=true, r=true}
  n8 = Add(A=Multiply(A=@n7) {constB=1.7}, B=Time())
  n9 = Ceil(None=Add(A=Sine(None=@n8)) {constB=0.7})
  n10 = Multiply(A=NamedRerouteUsage(), B=AppendVector(A=ComponentMask(None=@n9) {g=true}, B=ComponentMask(None=@n9) {r=true}))
  n11 = TextureCoordinate()
  n12 = Subtract(A=Add(A=ComponentMask(None=@n11) {g=true, r=true}) {constB=0.004}, B=Constant() {r=0.5})
  n13 = Abs(None=@n12)
  n14 = ComponentMask(None=@n13) {r=true}
  n15 = ComponentMask(None=@n13) {g=true}
  n16 = Clamp(
   None=Clamp(None=ComponentMask(None=@n10) {r=true}, Min=ComponentMask(None=@n10) {g=true}) {clampMode=CMODE_ClampMin}
   Min=MaterialFunctionCall!(
    Distance=Clamp(None=@n14, Min=@n15) {clampMode=CMODE_ClampMin}
    Radius=Constant() {r=0.495}
    ThicknessMultiplicator=Constant() {r=2}
    LocalVector=@n12
   ) {materialFunction=&/MF_ThicknessOutline}.Outline
  ) {clampMode=CMODE_ClampMin}
  n17 = Constant() {r=1}
  FeatureLevelSwitch(Default=@n16, ES2=@n17, ES3_1=@n17, SM4_REMOVED=@n16)
  n18 = Transform(None=Constant3Vector() {constant=<rgba 0 0 1 0>})
  n19 = CameraPositionWS()
  n20 = WorldPosition()
  n21 = Add(A=Normalize(VectorInput=Subtract(A=@n19, B=@n20.XYZ)), B=Constant3Vector() {constant=<rgba 0 0 5 0>})
  n22 = LinearInterpolate(Alpha=DotProduct(A=Normalize(VectorInput=@n18), B=Normalize(VectorInput=@n21))) {constA=0.5}
  Multiply(A=Multiply(A=Power(Base=@n22) {constExponent=100}, B=VectorParameter() {parameterName=WireColor, defaultValue=<rgba 0.8477 0.088 0 0>}.RGB)) {constB=5}
  n23 = Abs(None=Multiply(A=@n7))
  NamedRerouteDeclaration(None=@n23) {name=LocalVector, nodeColor=<rgba 0.2656 0 1 1>}
  NamedRerouteDeclaration(None=Fmod(A=@n23, B=Constant() {r=1})) {name=Distance, nodeColor=<rgba 0 0.7187 1 1>}
  NamedRerouteDeclaration(None=Constant() {r=0.5}) {name=Radius, nodeColor=<rgba 1 0 0.0469 1>}
  n24 = NamedRerouteUsage()
  n25 = NamedRerouteUsage()
  n26 = NamedRerouteUsage()
  n27 = MaterialFunctionCall(Vector 2=DDX(Value=@n26)) {materialFunction=/Engine/Functions/Engine_MaterialFunctions02/Utility/VectorLength}
  n28 = MaterialFunctionCall(Vector 2=DDY(Value=@n26)) {materialFunction=&VectorLength}
  n29 = Divide(A=Abs(None=Subtract(A=@n24, B=@n25)), B=Multiply(A=Min(A=@n27.V2 Length, B=@n28.V2 Length)) {constB=1.3})
  NamedRerouteDeclaration(None=OneMinus(None=Power(Base=@n29))) {name=Thickness2, nodeColor=<rgba 0 1 0.4141 1>}
  n30 = Constant3Vector() {constant=<rgba 0 0 1 0>}
  n31 = Distance(A=ComponentMask(None=Transform(None=@n30)) {g=true, r=true}, B=Constant2Vector())
  n32 = WorldPosition()
  n33 = CameraPositionWS()
  n34 = Divide(A=Multiply(A=Add(A=@n31) {constB=0.3}, B=Distance(A=@n32.XYZ, B=@n33))) {constB=5000000}
  Transform(None=Multiply(A=Constant3Vector() {constant=<rgba 0 0 1 0>}, B=Clamp(None=@n34)))
 ]
}
