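# Unlit, two-sided additive editor material that draws a height-colored glow for the Landscape gizmo.
# Local Z, normalized between Bottom and Top parameters, indexes a lookup texture, scaled by the view angle against world up.
# A depth comparison fades the glow to black as the scene recedes behind the pixel, and AlphaTexture further masks it toward black.
# A feature-level switch substitutes a flat constant on ES2 and ES3_1.
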Material /Engine/EditorLandscapeResources/LandscapeGizmoHeight_UnderMat
graph {
 properties {
  bUsedWithStaticLighting=true
  blendMode=BLEND_Additive
  pixelDepthOffsetMode=PDOM_Legacy
  shadingModel=MSM_Unlit
  twoSided=true
  usedShadingModels=Unlit
 }
 nodes [
  n1 = DotProduct(A=CameraVectorWS(), B=Constant3Vector() {constant=<rgba 0 0 1 0>})
  Abs(None=@n1)
  n2 = ComponentMask(None=Transform(None=WorldPosition().XYZ) {transformSourceType=TRANSFORMSOURCE_World, transformType=TRANSFORM_Local}) {b=true}
  n3 = ScalarParameter() {parameterName=Bottom}
  n4 = ScalarParameter() {defaultValue=300, parameterName=Top}
  n5 = AppendVector(A=Divide(A=Subtract(A=@n2, B=@n3), B=Subtract(A=@n4, B=@n3)), B=Constant())
  n6 = Constant() {r=1}
  n7 = Power(Base=@n1, Exp=Constant() {r=0.002})
  n8 = Clamp(None=Subtract(A=Constant() {r=0.999}, B=@n7), Min=Constant())
  n9 = Multiply(
   A=TextureSample(UVs=Multiply(A=@n5, B=@n6)) {texture=&/LandscapeGizmoHeightLookup}.RGB
   B=Multiply(A=Constant() {r=120}, B=@n8)
  )
  n10 = Constant3Vector()
  n11 = SceneDepth()
  n12 = PixelDepth()
  n13 = Clamp(None=Multiply(A=Subtract(A=@n11.R, B=@n12.R), B=Constant() {r=0.05}))
  n14 = LinearInterpolate(A=@n9, B=@n10, Alpha=@n13)
  Add(A=@n14, B=OneMinus(None=@n13))
  Add(A=@n9, B=Multiply(A=@n8, B=Constant() {r=2}))
  n15 = Fresnel() {exponent=1}
  Add(A=Multiply(A=OneMinus(None=@n15), B=Constant() {r=0.5}), B=Constant() {r=0.5})
  n16 = VectorParameter() {parameterName=AlphaScaleBias, defaultValue=<rgba 1 1 0 0>}
  n17 = AppendVector(A=@n16.RGB, B=@n16.RGB)
  n18 = TextureCoordinate()
  n19 = LinearInterpolate(
   A=@n14
   B=@n10
   Alpha=TextureSampleParameter2D(UVs=Multiply(A=@n17, B=@n18)) {parameterName=AlphaTexture, texture=&/DefaultAlphaTexture}.RGB
  )
  n20 = Constant() {r=1}
  FeatureLevelSwitch(Default=@n19, ES2=@n20, ES3_1=@n20, SM4_REMOVED=@n19)
  OneMinus(None=@n7)
 ]
}
