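# Masked, unlit, two-sided editor material for the landscape gizmo, which draws a square outline.
# Centered UVs give a square distance field, using a min-only Clamp as the max of absolute U and V.
# MF_ThicknessOutline turns that distance into the outline, and a feature level switch substitutes a constant 1 on ES2 and ES3.1.
# A constant red vector is also defined.
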Material /Engine/EditorLandscapeResources/LandscapeGizmo_Mat
graph {
 properties {
  bUsedWithStaticLighting=true
  blendMode=BLEND_Masked
  pixelDepthOffsetMode=PDOM_Legacy
  shadingModel=MSM_Unlit
  twoSided=true
  usedShadingModels=Unlit
 }
 nodes [
  Constant3Vector() {constant=<rgba 0.8477 0.088 0 0>}
  n1 = Subtract(A=ComponentMask(None=TextureCoordinate()) {g=true, r=true}, B=Constant() {r=0.5})
  n2 = Abs(None=ComponentMask(None=@n1) {g=true, r=true})
  n3 = MaterialFunctionCall!(
   Distance=Clamp(None=ComponentMask(None=@n2) {r=true}, Min=ComponentMask(None=@n2) {g=true}) {clampMode=CMODE_ClampMin}
   Radius=Constant() {r=0.5}
   ThicknessMultiplicator=Constant() {r=2}
   LocalVector=@n1
  ) {materialFunction=&/MF_ThicknessOutline}
  n4 = Constant() {r=1}
  FeatureLevelSwitch(Default=@n3.Outline, ES2=@n4, ES3_1=@n4, SM4_REMOVED=@n3.Outline)
 ]
}
