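# Unlit editor material that visualizes landscape layer usage per component as colored diagonal stripes with a count label.
# Local position over ComponentSizeVerts gives component coordinates, whose outer margin is masked to black as a border.
# A rotated, wrapped coordinate runs through a nested If chain against fractions of NumStripes, selecting among sixteen Color parameters.
# The half-brightness stripes are added to DebugIntValues, which prints NumStripes on tiled, center-scaled UVs.
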Material /Engine/EditorLandscapeResources/LandscapeLayerUsageMaterial
graph {
 properties {bUsedWithStaticLighting=true, pixelDepthOffsetMode=PDOM_Legacy, shadingModel=MSM_Unlit, usedShadingModels=Unlit}
 nodes [
  n1 = ScalarParameter() {defaultValue=64, parameterName=ComponentSizeVerts}
  n2 = Divide(
   A=MaterialFunctionCall() {materialFunction=&LocalPosition}.Local Position
   B=@n1
  ) {constB=64}
  n3 = ComponentMask(None=@n2) {r=true}
  n4 = ComponentMask(None=@n2) {g=true}
  n5 = Max(A=@n3, B=@n4)
  n6 = Divide(B=@n1)
  n7 = OneMinus(None=@n6)
  n8 = Constant()
  n9 = ComponentMask(None=@n2) {g=true, r=true}
  n10 = ScalarParameter() {defaultValue=3.1416, parameterName=Rotation}
  n11 = ComponentMask(None=Frac(None=Rotator(Coordinate=@n9, Time=@n10))) {r=true}
  n12 = ScalarParameter() {defaultValue=7, parameterName=NumStripes, sliderMax=10}
  n13 = Divide(B=@n12) {constA=2}
  n14 = Divide(B=@n12) {constA=5, constB=3}
  n15 = Divide(B=@n12) {constA=8, constB=3}
  n16 = Divide(B=@n12) {constA=11, constB=3}
  n17 = Divide(B=@n12) {constA=14, constB=3}
  n18 = VectorParameter() {parameterName=Color14, defaultValue=<rgba 0.5 0.5 1 1>}
  n19 = If(
   A=@n11
   B=Divide(B=@n12) {constA=15, constB=3}
   "A > B"=VectorParameter() {parameterName=Color15, defaultValue=<rgba 0.5 0.5 1 1>}.RGB
   "A == B"=@n18.RGB
   "A < B"=@n18.RGB
  )
  n20 = VectorParameter() {parameterName=Color13, defaultValue=<rgba 0.5 0.5 1 1>}
  n21 = VectorParameter() {parameterName=Color12, defaultValue=<rgba 0.5 0.5 1 1>}
  n22 = VectorParameter() {parameterName=Color11, defaultValue=<rgba 0.5 0.5 1 1>}
  n23 = If(
   A=@n11
   B=Divide(B=@n12) {constA=12, constB=3}
   "A > B"=If(
    A=@n11
    B=Divide(B=@n12) {constA=13, constB=3}
    "A > B"=If(A=@n11, B=@n17, "A > B"=@n19, "A == B"=@n20.RGB, "A < B"=@n20.RGB)
    "A == B"=@n21.RGB
    "A < B"=@n21.RGB
   )
   "A == B"=@n22.RGB
   "A < B"=@n22.RGB
  )
  n24 = VectorParameter() {parameterName=Color10, defaultValue=<rgba 0.5 0.5 1 1>}
  n25 = VectorParameter() {parameterName=Color9, defaultValue=<rgba 0.5 0.5 1 1>}
  n26 = VectorParameter() {parameterName=Color8, defaultValue=<rgba 0.5 1 0.5 1>}
  n27 = If(
   A=@n11
   B=Divide(B=@n12) {constA=9, constB=3}
   "A > B"=If(
    A=@n11
    B=Divide(B=@n12) {constA=10, constB=3}
    "A > B"=If(A=@n11, B=@n16, "A > B"=@n23, "A == B"=@n24.RGB, "A < B"=@n24.RGB)
    "A == B"=@n25.RGB
    "A < B"=@n25.RGB
   )
   "A == B"=@n26.RGB
   "A < B"=@n26.RGB
  )
  n28 = VectorParameter() {parameterName=Color7, defaultValue=<rgba 1 0.5 0.5 1>}
  n29 = VectorParameter() {parameterName=Color6, defaultValue=<rgba 1 1 1 1>}
  n30 = VectorParameter() {parameterName=Color5, defaultValue=<rgba 1 1 0 1>}
  n31 = If(
   A=@n11
   B=Divide(B=@n12) {constA=6, constB=3}
   "A > B"=If(
    A=@n11
    B=Divide(B=@n12) {constA=7, constB=3}
    "A > B"=If(A=@n11, B=@n15, "A > B"=@n27, "A == B"=@n28.RGB, "A < B"=@n28.RGB)
    "A == B"=@n29.RGB
    "A < B"=@n29.RGB
   )
   "A == B"=@n30.RGB
   "A < B"=@n30.RGB
  )
  n32 = VectorParameter() {parameterName=Color4, defaultValue=<rgba 0 1 1 1>}
  n33 = VectorParameter() {parameterName=Color3, defaultValue=<rgba 1 0 1 1>}
  n34 = VectorParameter() {parameterName=Color2, defaultValue=<rgba 0 0 1 1>}
  n35 = If(
   A=@n11
   B=Divide(B=@n12) {constA=3, constB=3}
   "A > B"=If(
    A=@n11
    B=Divide(B=@n12) {constA=4, constB=3}
    "A > B"=If(A=@n11, B=@n14, "A > B"=@n31, "A == B"=@n32.RGB, "A < B"=@n32.RGB)
    "A == B"=@n33.RGB
    "A < B"=@n33.RGB
   )
   "A == B"=@n34.RGB
   "A < B"=@n34.RGB
  )
  n36 = VectorParameter() {parameterName=Color1, defaultValue=<rgba 0 1 0 1>}
  n37 = VectorParameter() {parameterName=Color0, defaultValue=<rgba 1 0 0 1>}
  n38 = If(
   A=Min(A=@n3, B=@n4)
   B=@n6
   "A > B"=If(A=@n11, B=Divide(B=@n12), "A > B"=If(A=@n11, B=@n13, "A > B"=@n35, "A == B"=@n36.RGB, "A < B"=@n36.RGB), "A == B"=@n37.RGB, "A < B"=@n37.RGB)
   "A == B"=@n8
   "A < B"=@n8
  )
  n39 = Frac(None=Multiply(A=@n9) {constB=2})
  n40 = ComponentMask(None=VectorParameter() {parameterName=TextScale, defaultValue=<rgba 4 2 0 0>}.RGB) {g=true, r=true}
  Add(
   A=Multiply(A=If(A=@n5, B=@n7, "A > B"=@n8, "A == B"=@n8, "A < B"=@n38)) {constB=0.5}
   B=MaterialFunctionCall!(
    Number=@n12
    MaximumNumberOfDigits
    DebugTextLocation RG_UpperRight BA_LowerLeft
    UVs=MaterialFunctionCall(UVs=@n39, Texture Scale=@n40) {materialFunction=&ScaleUVsByCenter}.UVs
   ) {materialFunction=/Engine/Functions/Engine_MaterialFunctions01/Debug/DebugIntValues}.Result
  )
 ]
}
