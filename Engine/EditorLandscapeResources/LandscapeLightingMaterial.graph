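# LandscapeLightingMaterial is an engine editor-resource material for landscapes, apparently a neutral surface for viewing lighting.
# It feeds a constant mid-grey into the DiffuseColor input of the ConvertFromDiffSpec material function, leaving other function inputs at their defaults.
# It uses the Default Lit shading model and is flagged for use with static lighting.
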Material /Engine/EditorLandscapeResources/LandscapeLightingMaterial
graph {
 properties {bUsedWithStaticLighting=true, pixelDepthOffsetMode=PDOM_Legacy, usedShadingModels=Default Lit}
 nodes [MaterialFunctionCall(DiffuseColor=Constant() {r=0.5}) {materialFunction=&ConvertFromDiffSpec}]
}
