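# Unlit editor material that, going by its name, writes landscape heightmap data into a 2D render target.
# It samples a Heightmap texture parameter and branches on the ExportHeightIntoRGChannel scalar.
# At one or above, the sampled RGB passes through with height still packed in its channels.
# Below one, red and green are recombined as high and low bytes into a single unnormalized 16-bit height value.
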Material /Engine/EditorLandscapeResources/Landscape_Heightmap_To_RenderTarget2D
graph {
 properties {pixelDepthOffsetMode=PDOM_Legacy, shadingModel=MSM_Unlit, usedShadingModels=Unlit}
 nodes [
  n1 = TextureSampleParameter2D() {parameterName=Heightmap}
  n2 = Multiply(A=@n1.R) {constB=255}
  If(
   A=ScalarParameter() {parameterName=ExportHeightIntoRGChannel}
   "A > B"=@n1.RGB
   "A == B"=@n1.RGB
   "A < B"=Add(A=Multiply(A=@n2) {constB=256}, B=Multiply(A=@n1.G) {constB=255})
  ) {constB=1}
 ]
}
