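# Unlit editor material for landscape layer visualization, showing chosen texture channels as red, green and blue.
# For each color, a texture parameter sampled at landscape weightmap UVs is dot-multiplied with a vector mask parameter to pick its channel.
# The three dot products are appended into one RGB value.
# A feature level switch replaces the weightmap UVs with a constant on ES2.
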Material /Engine/EditorLandscapeResources/LayerVisMaterial
graph {
 properties {bUsedWithStaticLighting=true, pixelDepthOffsetMode=PDOM_Legacy, shadingModel=MSM_Unlit, usedShadingModels=Unlit}
 nodes [
  n1 = VectorParameter() {parameterName=Landscape_RedMask, defaultValue=<rgba 1 0 0 0>}
  n2 = AppendVector(A=@n1.RGB, B=@n1.RGB)
  n3 = LandscapeLayerCoords() {customUVType=LCCT_WeightMapUV}
  n4 = FeatureLevelSwitch(Default=@n3, ES2=Constant(), SM4_REMOVED=@n3)
  n5 = TextureSampleParameter2D(UVs=@n4) {parameterName=Landscape_RedTexture, texture=&/WhiteSquareTexture}
  n6 = AppendVector(A=@n5.RGB, B=@n5.RGB)
  n7 = VectorParameter() {parameterName=Landscape_GreenMask}
  n8 = AppendVector(A=@n7.RGB, B=@n7.RGB)
  n9 = TextureSampleParameter2D(UVs=@n4) {parameterName=Landscape_GreenTexture, texture=&WhiteSquareTexture}
  n10 = AppendVector(A=@n9.RGB, B=@n9.RGB)
  n11 = VectorParameter() {parameterName=Landscape_BlueMask}
  n12 = TextureSampleParameter2D(UVs=@n4) {parameterName=Landscape_BlueTexture, texture=&WhiteSquareTexture}
  AppendVector(
   A=AppendVector(A=DotProduct(A=@n2, B=@n6), B=DotProduct(A=@n8, B=@n10))
   B=DotProduct(A=AppendVector(A=@n11.RGB, B=@n11.RGB), B=AppendVector(A=@n12.RGB, B=@n12.RGB))
  )
  Constant()
 ]
}
