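# Unlit translucent landscape editor material that visualizes a mask, switched by the static Gizmo parameter between a gizmo alpha and selection data.
# The gizmo branch projects AlphaTexture from world position relative to LandscapeLocation, rotated by its W and scaled by AlphaScaleBias, masked to range and dimmed.
# The other branch samples SelectedData at landscape weightmap UVs, optionally inverted by bInverted and clamped to a floor.
# The switched result is inverted for the final output.
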Material /Engine/EditorLandscapeResources/MaskBrushMaterial
graph {
 properties {bUsedWithStaticLighting=true, blendMode=BLEND_Translucent, pixelDepthOffsetMode=PDOM_Legacy, shadingModel=MSM_Unlit, usedShadingModels=Unlit}
 nodes [
  n1 = VectorParameter() {parameterName=AlphaScaleBias, defaultValue=<rgba 0.0001 0.0001 0 0>}
  n2 = AppendVector(A=@n1.B, B=@n1.B)
  n3 = DoubleVectorParameter() {parameterName=LandscapeLocation, defaultValue=<0 0 0 0>}
  n4 = Subtract(A=WorldPosition().XYZ, B=@n3.XYZ)
  n5 = Rotator(
   Coordinate=ComponentMask(None=Transform(None=@n4) {transformSourceType=TRANSFORMSOURCE_World, transformType=TRANSFORM_Local}) {g=true, r=true}
   Time=@n3.W
  ) {centerX=0, centerY=0, speed=1}
  n6 = AppendVector(A=@n1.R, B=@n1.R)
  n7 = Add(A=Multiply(A=@n2, B=Multiply(A=@n5, B=@n6)), B=Multiply(A=@n2, B=Constant2Vector() {g=0.5, r=0.5}))
  n8 = MaterialFunctionCall!(RangeMin, RangeMax=@n2, Value=@n7) {materialFunction=/Engine/Functions/Engine_MaterialFunctions03/Math/WithinRangeFloat2}
  n9 = TextureSampleParameter2D(UVs=@n7) {parameterName=AlphaTexture, texture=&/DefaultAlphaTexture}
  n10 = Constant() {r=0.5}
  n11 = Multiply(A=OneMinus(None=Multiply(A=@n8.AllComponentsWithinRange, B=@n9.RGB)), B=@n10)
  n12 = OneMinus(None=@n10)
  n13 = LandscapeLayerCoords() {customUVType=LCCT_WeightMapUV}
  n14 = TextureSampleParameter2D(UVs=FeatureLevelSwitch(Default=@n13, ES2=Constant(), SM4_REMOVED=@n13)) {parameterName=SelectedData, texture=&DefaultAlphaTexture}
  n15 = If(A=ScalarParameter() {parameterName=bInverted}, "A > B"=OneMinus(None=@n14.R), "A == B"=@n14.R, "A < B"=@n14.R)
  n16 = MaterialFunctionCall(X=@n15, Y=@n15, Z=@n15) {materialFunction=&MakeFloat3}
  OneMinus(None=StaticSwitchParameter(True=Add(A=@n11, B=@n12), False=Clamp(None=@n16.Result) {minDefault=0.35}) {parameterName=Gizmo})
 ]
}
