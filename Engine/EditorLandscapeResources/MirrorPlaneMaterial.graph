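# Unlit, translucent, two-sided editor material that, by its name and path, draws the landscape mirror plane as a green grid.
# It scales UVs by the GridSize parameter, then offsets them by DDX and DDY times GridThickness, keeping line width steady in screen pixels.
# Floor differences between offset and original coordinates detect cell boundaries, and an If gives lines a higher value than cell interiors.
# Feature-level switches substitute the scaled UVs for derivatives on ES2.
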Material /Engine/EditorLandscapeResources/MirrorPlaneMaterial
graph {
 properties {
  blendMode=BLEND_Translucent
  pixelDepthOffsetMode=PDOM_Legacy
  shadingModel=MSM_Unlit
  translucencyPass=MTP_BeforeDOF
  twoSided=true
  usedShadingModels=Unlit
 }
 nodes [
  Constant3Vector() {constant=<rgba 0.01 1 0.01 0>}
  n1 = Multiply(A=TextureCoordinate(), B=ComponentMask(None=VectorParameter() {parameterName=GridSize, defaultValue=<rgba 10 10 0 0>}.RGB) {g=true, r=true})
  n2 = DDX(Value=@n1)
  n3 = ScalarParameter() {defaultValue=1, parameterName=GridThickness}
  n4 = Add(A=@n1, B=Multiply(A=FeatureLevelSwitch(Default=@n2, ES2=@n1, SM4_REMOVED=@n2), B=@n3))
  n5 = Floor(None=@n1)
  n6 = Subtract(A=Floor(None=@n4), B=@n5)
  n7 = DotProduct(A=@n6, B=@n6)
  n8 = DDY(Value=@n1)
  n9 = Multiply(A=FeatureLevelSwitch(Default=@n8, ES2=@n1, SM4_REMOVED=@n8), B=@n3)
  n10 = Subtract(A=Floor(None=Add(A=@n1, B=@n9)), B=@n5)
  n11 = DotProduct(A=@n10, B=@n10)
  n12 = Subtract(A=Floor(None=Add(A=@n4, B=@n9)), B=@n5)
  n13 = Constant() {r=0.05}
  If(A=Add(A=Add(A=@n7, B=@n11), B=DotProduct(A=@n12, B=@n12)), "A > B"=Constant() {r=0.5}, "A == B"=@n13, "A < B"=@n13)
 ]
}
