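# Translucent unlit editor overlay, drawn without depth testing, that highlights landscape selection or brush area in a pulsing HighlightColor tint.
# The Selected switch chooses between the SelectedData mask on UV1 and a brush shape, either flat or AlphaTexture.
# Brush alpha is projected from local position relative to LandscapeLocation, rotated by its W, scaled and range-clipped.
# An absolute time sine pulses the intensity, while low nonzero selection values brighten the color.
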Material /Engine/EditorLandscapeResources/SelectBrushMaterial
graph {
 properties {
  bDisableDepthTest=true
  bUsedWithStaticLighting=true
  blendMode=BLEND_Translucent
  pixelDepthOffsetMode=PDOM_Legacy
  shadingModel=MSM_Unlit
  translucencyPass=MTP_BeforeDOF
  usedShadingModels=Unlit
 }
 nodes [
  n1 = TextureSampleParameter2D(UVs=TextureCoordinate() {coordinateIndex=1}) {parameterName=SelectedData, texture=&/WhiteSquareTexture}
  n2 = Time() {bIgnorePause=true}
  n3 = Constant() {r=0.5}
  n4 = Abs(None=Sine(None=Multiply(A=@n2, B=@n3)))
  n5 = Multiply(A=@n1.RGB, B=@n4)
  n6 = VectorParameter() {parameterName=AlphaScaleBias, defaultValue=<rgba 0.0001 0.0001 0 0>}
  n7 = AppendVector(A=@n6.B, B=@n6.B)
  n8 = DoubleVectorParameter() {parameterName=LandscapeLocation, defaultValue=<0 0 0 0>}
  n9 = Subtract(A=WorldPosition().XYZ, B=@n8.XYZ)
  n10 = Rotator(
   Coordinate=ComponentMask(None=Transform(None=@n9) {transformSourceType=TRANSFORMSOURCE_World, transformType=TRANSFORM_Local}) {g=true, r=true}
   Time=@n8.W
  ) {centerX=0, centerY=0, speed=1}
  n11 = AppendVector(A=@n6.R, B=@n6.R)
  n12 = Add(A=Multiply(A=@n7, B=Multiply(A=@n10, B=@n11)), B=Multiply(A=@n7, B=Constant2Vector() {g=0.5, r=0.5}))
  n13 = Multiply(
   A=TextureSampleParameter2D(UVs=@n12) {parameterName=AlphaTexture, texture=&/DefaultAlphaTexture}.RGB
   B=MaterialFunctionCall!(RangeMin, RangeMax=@n7, Value=@n12) {materialFunction=/Engine/Functions/Engine_MaterialFunctions03/Math/WithinRangeFloat2}.AllComponentsWithinRange
  )
  n14 = Constant() {r=0.5}
  n15 = StaticSwitchParameter(True=@n5, False=Multiply(A=StaticSwitchParameter(True=@n13, False=@n14) {parameterName=AlphaBrush}, B=@n4)) {parameterName=Selected}
  n16 = Constant() {r=0.075}
  n17 = Constant() {r=0.1}
  Add(
   A=StaticSwitchParameter(True=Constant(), False=Constant() {r=0.4}) {parameterName=Selected}
   B=Add(A=Multiply(A=@n15, B=@n16), B=Multiply(A=@n1.RGB, B=@n17))
  )
  n18 = Subtract(A=@n1.R) {constB=0.2}
  n19 = OneMinus(None=Max(A=Sign(None=@n18)) {constB=0})
  n20 = Max(A=Sign(None=ComponentMask(None=@n5) {r=true})) {constB=0}
  n21 = ComponentMask(None=VectorParameter() {parameterName=HighlightColor, defaultValue=<rgba 1 0.1577 0 1>}.RGB) {b=true, g=true, r=true}
  StaticSwitchParameter(True=Add(A=Multiply(A=@n19, B=@n20), B=@n21), False=@n21) {parameterName=Selected}
 ]
}
