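# Unlit, two-sided editor material for landscape spline meshes, drawing bright green bands over a dim green base.
# It detects where the V texture coordinate crosses whole numbers by comparing floored values at derivative-offset samples.
# Screen-space DDX and DDY scaled by the LineWidth parameter keep line thickness roughly constant in pixels.
# Feature level switches replace both the line intensity and the tinted color with constant one on ES2 and ES3.1.
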Material /Engine/EditorLandscapeResources/SplineEditorMeshMat
graph {
 properties {
  bUsedWithNanite=true
  bUsedWithSplineMeshes=true
  pixelDepthOffsetMode=PDOM_Legacy
  shadingModel=MSM_Unlit
  twoSided=true
  usedShadingModels=Unlit
 }
 nodes [
  n1 = ComponentMask(None=TextureCoordinate()) {g=true}
  n2 = ScalarParameter() {defaultValue=2, parameterName=LineWidth}
  n3 = Multiply(A=DDX(Value=@n1), B=@n2)
  n4 = Add(A=@n1, B=@n3)
  n5 = Multiply(A=DDY(Value=@n1), B=@n2)
  n6 = Subtract(A=@n1, B=@n3)
  n7 = Subtract(A=Floor(None=Add(A=@n4, B=@n5)), B=Floor(None=Subtract(A=@n6, B=@n5)))
  n8 = Subtract(A=@n1, B=@n3)
  n9 = Add(A=@n1, B=@n3)
  n10 = Subtract(A=Floor(None=Add(A=@n8, B=@n5)), B=Floor(None=Subtract(A=@n9, B=@n5)))
  n11 = If(A=Max(A=Abs(None=@n7), B=Abs(None=@n10)), B=Constant() {r=1}, "A > B"=Constant() {r=1}, "A < B"=Constant() {r=0.15})
  n12 = Constant() {r=1}
  FeatureLevelSwitch(Default=@n11, ES2=@n12, ES3_1=@n12, SM4_REMOVED=@n11)
  n13 = Multiply(A=Constant3Vector() {constant=<rgba 0.35 0.9 0 0>}, B=@n11)
  n14 = Constant() {r=1}
  FeatureLevelSwitch(Default=@n13, ES2=@n14, ES3_1=@n14, SM4_REMOVED=@n13)
 ]
}
